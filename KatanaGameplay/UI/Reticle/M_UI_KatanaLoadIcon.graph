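# Translucent UI material that draws the Katana load icon for the reticle as a procedural six-sided shape, without textures.
# MF_6VertexPolygon builds fill and outline masks from six hard-coded vertices, shared through named reroutes.
# Fill and outline colors each have an optional linear gradient via static switches, then blend into the final color.
# An EnableOutline scalar fades the outline in color and alpha, and an Opacity scalar scales overall alpha.
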Material /KatanaGameplay/UI/Reticle/M_UI_KatanaLoadIcon
graph {
 properties {blendMode=BLEND_Translucent, materialDomain=MD_UI, pixelDepthOffsetMode=PDOM_Legacy, usedShadingModels=Default Lit}
 nodes [
  n1 = VectorParameter() {group=Shape, parameterName=ShapeColor1, defaultValue=<rgba 0.0545 0.8796 0.8796 1>}
  n2 = MaterialFunctionCall() {materialFunction=&LinearGradient}
  n3 = StaticSwitchParameter(
   True=LinearInterpolate(
    A=@n1.RGB
    B=VectorParameter() {group=Shape, parameterName=ShaperColor2, defaultValue=<rgba 1 0.0742 0 1>}.RGB
    Alpha=Power(Base=@n2.VGradient) {constExponent=1.3}
   )
   False=@n1.RGB
  ) {group=Shape, parameterName=Shape Gradient, sortPriority=1}
  n4 = NamedRerouteUsage()
  n5 = VectorParameter() {group=Outline, parameterName=OutlineColor1, defaultValue=<rgba 1 0 0.5583 1>}
  n6 = StaticSwitchParameter(
   True=LinearInterpolate(
    A=@n5.RGB
    B=VectorParameter() {group=Outline, parameterName=OutlineColor2, defaultValue=<rgba 0.8177 0.1782 0 1>}.RGB
    Alpha=MaterialFunctionCall() {materialFunction=&LinearGradient}.VGradient
   )
   False=@n5.RGB
  ) {group=Outline, parameterName=Outline Gradient, sortPriority=1}
  LinearInterpolate(
   A=Multiply(A=NamedRerouteUsage(), B=@n3)
   B=LinearInterpolate(A=@n3, B=Multiply(A=@n4, B=@n6), Alpha=ScalarParameter() {defaultValue=1, parameterName=EnableOutline})
   Alpha=NamedRerouteUsage()
  )
  n7 = NamedRerouteUsage()
  n8 = ScalarParameter() {defaultValue=1, parameterName=EnableOutline}
  Multiply(A=Add(A=NamedRerouteUsage(), B=LinearInterpolate(B=@n7, Alpha=@n8)), B=ScalarParameter() {defaultValue=1, parameterName=Opacity})
  n9 = MaterialFunctionCall!(
   UVs
   vertex_0=NamedRerouteUsage()
   vertex_1=NamedRerouteUsage()
   vertex_2=NamedRerouteUsage()
   vertex_4=NamedRerouteUsage()
   vertex_3=NamedRerouteUsage()
   vertex_5=NamedRerouteUsage()
   Outline Thickness=Constant() {r=0.02}
   Glow Min
   Glow Max=Constant() {r=0.1}
  ) {materialFunction=&/MF_6VertexPolygon}
  NamedRerouteDeclaration(None=@n9.Fill) {name=Shape, nodeColor=<rgba 0 0 0 1>}
  NamedRerouteDeclaration(None=@n9.Outline) {name=ShapeOutilne, nodeColor=<rgba 0 0 0 1>}
  n10 = Constant3Vector() {constant=<rgba -0.99 -0.3923 0 1>}
  NamedRerouteDeclaration(None=AppendVector(A=@n10, B=@n10)) {name=Vertex0, nodeColor=<rgba 0 0 0 1>}
  n11 = Constant3Vector() {constant=<rgba -0.99 -0.2304 0 1>}
  NamedRerouteDeclaration(None=AppendVector(A=@n11, B=@n11)) {name=Vertex1, nodeColor=<rgba 0 0 0 1>}
  n12 = Constant3Vector() {constant=<rgba 0 0.3857 0 1>}
  NamedRerouteDeclaration(None=AppendVector(A=@n12, B=@n12)) {name=Vertex2, nodeColor=<rgba 0 0 0 1>}
  n13 = Constant3Vector() {constant=<rgba 0.99 0.2301 0 1>}
  NamedRerouteDeclaration(None=AppendVector(A=@n13, B=@n13)) {name=Vertex3, nodeColor=<rgba 0 0 0 1>}
  n14 = Constant3Vector() {constant=<rgba 0.99 0.3901 0 1>}
  NamedRerouteDeclaration(None=AppendVector(A=@n14, B=@n14)) {name=Vertex4, nodeColor=<rgba 0 0 0 1>}
  n15 = Constant3Vector() {constant=<rgba 0 -0.3979 0 1>}
  NamedRerouteDeclaration(None=AppendVector(A=@n15, B=@n15)) {name=Vertex5, nodeColor=<rgba 0 0 0 1>}
 ]
}
