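# Translucent UI material for the Katana reticle, drawing procedural shapes that respond to lock-on and animation parameters.
# Shape 1 is a mirrored six-vertex polygon offset by LockAnim, with AnimBlur driving both its edge softness and color blend.
# Shape 2, an outlined polygon with glow and optional gradient, joins Shape 3, a rotated box, with LockedOn blending their color and mask.
# Named reroutes carry vertices, UVs, tints and masks into the merged result.
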Material /KatanaGameplay/UI/Reticle/M_UI_KatanaReticleBase
graph {
 properties {blendMode=BLEND_Translucent, materialDomain=MD_UI, pixelDepthOffsetMode=PDOM_Legacy, usedShadingModels=Default Lit}
 nodes [
  NamedRerouteDeclaration(None=Constant2Vector() {g=-0.5797, r=-0.1979}) {name=Shape's 1 Vertex 0, nodeColor=<rgba 0 0 0 1>}
  NamedRerouteDeclaration(None=Constant2Vector() {g=-0.491, r=-0.5157}) {name=Shape's 1 Vertex 1, nodeColor=<rgba 0 0 0 1>}
  n1 = MaterialFunctionCall!(
   UVs=NamedRerouteUsage()
   vertex_0=NamedRerouteUsage()
   vertex_1=NamedRerouteUsage()
   vertex_2=NamedRerouteUsage()
   vertex_4=NamedRerouteUsage()
   vertex_3=NamedRerouteUsage()
   vertex_5=NamedRerouteUsage()
   Outline Thickness=ScalarParameter() {group=Outline, parameterName=OutlineThickness}
   Glow Min=Constant()
   Glow Max=Constant() {r=0.05}
  ) {materialFunction=&/MF_6VertexPolygon}
  n2 = Subtract(
   A=Subtract(A=OneMinus(None=@n1.SDF Outline), B=ScalarParameter() {defaultValue=1, group=Glow, parameterName=GlowIntensity}) {constB=0.2}
   B=NamedRerouteDeclaration(None=@n1.Fill) {name=Custom Shape 2, nodeColor=<rgba 0 0 0 1>}
  )
  n3 = NamedRerouteDeclaration(None=@n1.Outline) {name=Custom Shape 2_Outline, nodeColor=<rgba 0 0 0 1>}
  NamedRerouteDeclaration(None=Saturate(None=Subtract(A=@n2, B=@n3))) {name=Custom Shape 2_Glow, nodeColor=<rgba 0 0 0 1>}
  NamedRerouteDeclaration(None=AppendVector(A=Constant() {r=0.6}, B=ScalarParameter() {defaultValue=0.3, parameterName=ScaleOfShapes})) {name=Shape's 1 Scale, nodeColor=<rgba 0 0 0 1>}
  NamedRerouteDeclaration(None=AppendVector(A=ScalarParameter() {defaultValue=0.1, parameterName=LockAnim}, B=Constant())) {name=Shape's 1 Offset, nodeColor=<rgba 0 0 0 1>}
  NamedRerouteDeclaration(None=LinearInterpolate(
    A=VectorParameter() {parameterName=NormalColor, defaultValue=<rgba 0.4154 0.651 0.651 1>}.RGB
    B=VectorParameter() {parameterName=ReticleColor, defaultValue=<rgba 1 1 1 1>}.RGB
    Alpha=ScalarParameter() {defaultValue=1, parameterName=AnimBlur}
   )) {name=Custom Shape 1 Color, nodeColor=<rgba 0 0 0 1>}
  NamedRerouteDeclaration(None=Multiply(A=NamedRerouteUsage(), B=NamedRerouteUsage())) {name=Tinted Shape 1, nodeColor=<rgba 0 0 0 1>}
  n4 = VectorParameter() {group=Shape, parameterName=ShapeColor1, defaultValue=<rgba 1 0 0.5583 1>}
  n5 = Subtract(A=TextureCoordinate(), B=Constant() {r=0.35}) {constB=0}
  n6 = StaticSwitchParameter(
   True=LinearInterpolate(
    A=@n4.RGB
    B=VectorParameter() {group=Shape, parameterName=ShapeColor2, defaultValue=<rgba 1 0.807 0 1>}.RGB
    Alpha=MaterialFunctionCall(UV Channel=@n5) {materialFunction=&LinearGradient}.VGradient
   )
   False=@n4.RGB
  ) {group=Shape, parameterName=Shape Gradient, sortPriority=1}
  n7 = NamedRerouteUsage()
  n8 = NamedRerouteUsage()
  n9 = Add(
   A=Multiply(A=Multiply(A=@n6, B=@n7)) {constB=9}
   B=Multiply(A=@n8, B=VectorParameter() {group=Outline, parameterName=OutlineColor, defaultValue=<rgba 0.2812 0.0358 0 1>}.RGB)
  )
  n10 = VectorParameter() {group=Glow, parameterName=GlowColor, defaultValue=<rgba 1 0.616 0.1198 1>}
  n11 = Saturate(None=Subtract(A=Subtract(A=@n10.RGB, B=@n7), B=@n8))
  NamedRerouteDeclaration(None=LinearInterpolate(A=Constant3Vector() {constant=<rgba 1 1 1 0>}, B=Add(A=@n9, B=@n11), Alpha=ScalarParameter() {defaultValue=1, parameterName=LockedOn}) {constAlpha=0}) {name="Custom Shape 2 && 3 Color", nodeColor=<rgba 0 0 0 1>}
  NamedRerouteDeclaration(None=Constant2Vector() {g=-0.4966, r=0.7547}) {name=Shape 1's Vertex 2, nodeColor=<rgba 0 0 0 1>}
  NamedRerouteDeclaration(None=Constant2Vector() {g=-0.4106, r=-1}) {name=Shape2 's Vertex 0, nodeColor=<rgba 0 0 0 1>}
  NamedRerouteDeclaration(None=Constant2Vector() {g=-0.6007, r=-0.7916}) {name=Shape2 's Vertex 1, nodeColor=<rgba 0 0 0 1>}
  NamedRerouteDeclaration(None=Constant2Vector() {g=0.1891, r=-0.0016}) {name=Shape2 's Vertex 2, nodeColor=<rgba 0 0 0 1>}
  NamedRerouteDeclaration(None=Constant2Vector() {g=-0.4007, r=1}) {name=Shape2 's Vertex 3, nodeColor=<rgba 0 0 0 1>}
  NamedRerouteDeclaration(None=Constant2Vector() {g=-0.611, r=0.7811}) {name=Shape2 's Vertex 4, nodeColor=<rgba 0 0 0 1>}
  NamedRerouteDeclaration(None=Constant2Vector() {g=0.6079, r=0.002}) {name=Shape2 's Vertex 5, nodeColor=<rgba 0 0 0 1>}
  n12 = TextureCoordinate()
  NamedRerouteDeclaration(None=ComponentMask(None=@n12) {g=true}) {name=S2's G Channel Mask, nodeColor=<rgba 0 0 0 1>}
  NamedRerouteDeclaration(None=Constant2Vector() {g=0.573, r=-0.6491}) {name=Shape 1's Vertex 3, nodeColor=<rgba 0 0 0 1>}
  NamedRerouteDeclaration(None=Add(A=NamedRerouteUsage()) {constB=0.15}) {name=S2's G Channel Offset, nodeColor=<rgba 0 0 0 1>}
  n13 = ComponentMask(None=@n12) {r=true}
  n14 = NamedRerouteUsage()
  NamedRerouteDeclaration(None=MaterialFunctionCall(UVs=AppendVector(A=@n13, B=@n14), Texture Scale=Constant() {r=0.14}) {materialFunction=&ScaleUVsByCenter}.UVs) {name=Shape' s Scaled UV, nodeColor=<rgba 0 0 0 1>}
  n15 = TextureCoordinate()
  NamedRerouteDeclaration(None=ComponentMask(None=@n15) {g=true}) {name=S3's G Channel Mask, nodeColor=<rgba 0 0 0 1>}
  NamedRerouteDeclaration(None=Add(A=NamedRerouteUsage()) {constB=0.15}) {name=S3's G Channel Offset, nodeColor=<rgba 0 0 0 1>}
  n16 = Add(A=NamedRerouteUsage(), B=NamedRerouteUsage())
  n17 = NamedRerouteUsage()
  n18 = LinearInterpolate(A=NamedRerouteUsage(), B=Saturate(None=Add(A=@n16, B=@n17)), Alpha=ScalarParameter() {defaultValue=1, parameterName=LockedOn}) {constAlpha=0}
  NamedRerouteDeclaration(None=Add(A=NamedRerouteUsage(), B=NamedRerouteDeclaration(None=@n18) {name="Custom Shapes 2 && 3", nodeColor=<rgba 0 0 0 1>})) {name=Merged Shapes, nodeColor=<rgba 0 0 0 1>}
  n19 = ScalarParameter() {defaultValue=1, parameterName=Opacity}
  NamedRerouteDeclaration(None=Multiply(A=NamedRerouteUsage(), B=Clamp(None=@n19) {maxDefault=0.7})) {name=Opacity driven S1, nodeColor=<rgba 0.5 1 0 1>}
  NamedRerouteDeclaration(None=Add(A=NamedRerouteUsage(), B=NamedRerouteUsage())) {name=Tinted Shapes, nodeColor=<rgba 0 0 0 1>}
  NamedRerouteDeclaration(None=Constant2Vector() {g=0.5913, r=-0.99}) {name=Shape 1's Vertex 4, nodeColor=<rgba 0 0 0 1>}
  NamedRerouteDeclaration(None=Constant2Vector() {g=-0.5761, r=0.99}) {name=Shape 1's Vertex 5, nodeColor=<rgba 0 0 0 1>}
  n20 = ScalarParameter() {defaultValue=0.75, parameterName="Blur(0 - 0.99)"}
  n21 = ScalarParameter() {defaultValue=1, parameterName=AnimBlur}
  n22 = TextureCoordinate()
  n23 = NamedRerouteUsage()
  n24 = NamedRerouteUsage()
  NamedRerouteDeclaration(None=MaterialFunctionCall(
    UVs=MaterialFunctionCall!(UV=@n22, Offset=@n23, Rotation, Scale=@n24, MirrorY) {materialFunction=/FortUILibrary/MaterialFunctions/Basics/MF_UI_Symmetry_2x1_DEPRECATED}.Result
    vertex_0=NamedRerouteUsage()
    vertex_1=NamedRerouteUsage()
    vertex_2=NamedRerouteUsage()
    vertex_4=NamedRerouteUsage()
    vertex_3=NamedRerouteUsage()
    vertex_5=NamedRerouteUsage()
    Glow Max=LinearInterpolate(A=@n20, Alpha=@n21) {constAlpha=0, constB=0.1}
   ) {materialFunction=&MF_6VertexPolygon}.Fill) {name=Custom Shape 1, nodeColor=<rgba 0 0 0 1>}
  n25 = MaterialFunctionCall(UVs=AppendVector(A=ComponentMask(None=@n15) {r=true}, B=NamedRerouteUsage()), Texture Scale=Constant() {r=0.08}) {materialFunction=&ScaleUVsByCenter}
  n26 = Constant() {r=0.125}
  NamedRerouteDeclaration(None=MaterialFunctionCall(
    UVs=MaterialFunctionCall(UVs=@n25.UVs, "Rotation Angle (0-1)"=@n26) {materialFunction=&CustomRotator}.Rotated Values
    GlowMax=Constant() {r=0.1}
   ) {materialFunction=&MF_UI_Box}.Fill) {name=Custom Shape 3, nodeColor=<rgba 0 0 0 1>}
  NamedRerouteDeclaration(None=Multiply(A=NamedRerouteUsage(), B=NamedRerouteUsage())) {name="Tinted Shape 2 && 3", nodeColor=<rgba 0 0 0 1>}
  NamedRerouteUsage()
  NamedRerouteUsage()
 ]
}
