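# Katana reticle pip widget, a child of the shared recharge-stack pip base, showing one charge as a single material-driven image.
# SetFillPercent remaps fill into a depleted-to-max opacity range before calling the parent, writing to the Opacity parameter by default.
# Init, fill and use overrides drive that fill between empty and full, with init also setting the material's EnableOutline.
# Gain and use animations stop each other and drive the same outline parameter.
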WidgetBlueprint /KatanaGameplay/UI/Reticle/WBP_Reticle_Katana_Pip
animations [
 Anim_Gain {
  binding_count=1
  length_seconds=0.75
  object=&self:Anim_Gain
  sequence {bindings [Image_Pip {
     id=#1
     possessed_class=Image
     tracks [{
       type=MovieSceneWidgetMaterialTrack
       sections [{type=MovieSceneParameterSection, end_frame=15, start_frame=0, channels [EnableOutline {type=MovieSceneScriptingFloatChannel, const [1,1]}]}]
      }]
    }]}
 }
 Anim_Use {
  binding_count=1
  length_seconds=0.5
  object=&self:Anim_Use
  sequence {bindings [Image_Pip {
     id=#2
     possessed_class=Image
     tracks [{
       type=MovieSceneWidgetMaterialTrack
       sections [{type=MovieSceneParameterSection, end_frame=10, start_frame=0, channels [EnableOutline {type=MovieSceneScriptingFloatChannel, const [0,1]}]}]
      }]
    }]}
 }
]
graphs [
 EventGraph {nodes [
   Event Construct()
   > SetImagePip!(
    execute
    self:Fort Weapon Reticle Recharge Stack Base Pip Base Object Reference
    Image:Image Object Reference=Get Image_Pip() -> (Image_Pip:Image Object Reference)
   ) -> (then)
   n1 = Event SetFillPercent!() -> (OutputDelegate, then, FillPercent:float, PipPercentMatParamName:Name=None) {is=title K2Node_Event}
   > Parent: SetFillPercent!(
    execute
    FillPercent:float=MapRangeClamped(
     Value=@n1.FillPercent
     InRangeA=0
     InRangeB=1
     OutRangeA=Get PipDepletedOpacity() -> (PipDepletedOpacity:double)
     OutRangeB=Get PipMaxOpacity() -> (PipMaxOpacity:double)
    )
    PipPercentMatParamName:Name=@n1.PipPercentMatParamName
   ) -> (then) {is=title K2Node_CallParentFunction}
  ]}
 InitFillState {nodes [
   n1 = InitFillState!() -> (then, IsFilled:bool) {is=title K2Node_FunctionEntry}
   > Parent: InitFillState!(execute, IsFilled:bool=@n1) -> (then) {is=title K2Node_CallParentFunction}
   > n2 = GetDynamicMaterial(self=Get Image_Pip())
   n3 = SelectFloat(A=1, B=0, bPickA=@n1)
   > SetScalarParameterValue(self=@n2, ParameterName=EnableOutline, Value=@n3)
   > SetFillPercent!(
    execute
    self:Fort Weapon Reticle Recharge Stack Base Pip Base Object Reference
    FillPercent:float=@n3
    PipPercentMatParamName:Name=Get FillPercentParameterName() -> (FillPercentParameterName:Name=None)
   ) -> (then)
  ]}
 OnFill {nodes [
   OnFill() {is=title K2Node_FunctionEntry}
   > Parent: OnFill() {is=title K2Node_CallParentFunction}
   > SetFillPercent(FillPercent=1, PipPercentMatParamName=Get FillPercentParameterName() -> (FillPercentParameterName=None))
  ]}
 OnUse {nodes [
   OnUse() {is=title K2Node_FunctionEntry}
   > Parent: OnUse() {is=title K2Node_CallParentFunction}
   > SetFillPercent(FillPercent=0, PipPercentMatParamName=Get FillPercentParameterName() -> (FillPercentParameterName=None))
  ]}
 PlayFillAnimation {nodes [
   PlayFillAnimation() {is=title K2Node_FunctionEntry}
   > Parent: PlayFillAnimation() {is=title K2Node_CallParentFunction}
   > StopAnimation(InAnimation=Get Anim_Use() -> (Anim_Use:Widget Animation Object Reference))
   > PlayAnimation(InAnimation=Get Anim_Gain() -> (Anim_Gain:Widget Animation Object Reference))
  ]}
 PlayUseAnimation {nodes [
   PlayUseAnimation() {is=title K2Node_FunctionEntry}
   > Parent: PlayUseAnimation() {is=title K2Node_CallParentFunction}
   > StopAnimation(InAnimation=Get Anim_Gain())
   > PlayAnimation(InAnimation=Get Anim_Use())
  ]}
 SetMaterial {nodes [
   n1 = SetMaterial!() -> (then, Material:Material Interface Object Reference) {is=title K2Node_FunctionEntry}
   > SetBrushFromMaterial(self=Get Image_Pip(), Material=@n1)
  ]}
]
variables [
 FillPercentParameterName {type=string, default=Opacity}
 PipDepletedOpacity {type=number, default=0.35}
 PipMaxOpacity {type=number, default=0.92}
]
widgets {
 info {
  inheritedWidgetCount=0
  parentClass=/Game/Athena/HUD/reticles/ReticleExtensionWidgets/ChargeStack/WBP_Reticle_RechargeStackBase_Pip.WBP_Reticle_RechargeStackBase_Pip_C
  rootWidgetClass=&Image
  widgetCount=1
 }
 widgets [Image_Pip {
   widgetClassPath=&Image
   bIsVariable=true
   widget=Image_Pip
   properties {brush {
     bIsDynamicallyLoaded=false
     drawAs=Image
     imageType=NoImage
     mirroring=NoMirror
     resourceName=None
     resourceObject=&/MI_UI_KatanaLoadIcon
     tiling=NoTile
     imageSize=<32 32>
     margin=<ltrb 0 0 0 0>
     outlineSettings {
      bUseBrushTransparency=false
      roundingType=HalfHeightRadius
      width=0
      color {colorUseRule=UseColor_Specified, specifiedColor=<rgba 0 0 0 0>}
      cornerRadii=<0 0 0 0>
     }
     tintColor {colorUseRule=UseColor_Specified, specifiedColor=<rgba 1 1 1 1>}
     uVRegion {bIsValid=false, max=<0 0>, min=<0 0>}
    }}
  }]
}
