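# Katana reticle widget that shows a collection of recharge stack pips, built on the shared RechargeStackBase pip collection.
# Its root Dynamic Entry Box spawns Katana pip entries and is handed to the parent class on Construct.
# SetPipMaterial only stores a material, which the OnStackCreated override applies to each newly created pip when valid, leaving existing pips untouched.
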WidgetBlueprint /KatanaGameplay/UI/Reticle/WBP_Reticle_Katana_PipCollection
graphs [
 EventGraph {nodes [
   Event Construct()
   > SetDynamicEntryBox!(
    execute
    self:Fort Weapon Reticle Recharge Stack Base Pip Collection Base Object Reference
    DynamicEntryBox:Dynamic Entry Box Object Reference=Get DynamicEntryBox_Stacks() -> (DynamicEntryBox_Stacks:Dynamic Entry Box Object Reference)
   ) -> (then)
  ]}
 OnStackCreated {nodes [
   n1 = OnStackCreated!() -> (then, StackWidget:WBP Reticle Recharge Stack Base Pip Object Reference) {is=title K2Node_FunctionEntry}
   > Parent: OnStackCreated!(execute, StackWidget:WBP Reticle Recharge Stack Base Pip Object Reference=@n1) -> (then) {is=title K2Node_CallParentFunction}
   > Is Valid(InputObject=Get PipMaterial() -> (PipMaterial:Material Interface Object Reference))
   > n2 = Cast To WBP_Reticle_Katana_Pip!(execute, Object=@n1) -> (then, CastFailed, AsWBP Reticle Katana Pip:WBP Reticle Katana Pip Object Reference) {is=title K2Node_DynamicCast}
   > SetMaterial!(execute, self:WBP Reticle Katana Pip Object Reference=@n2, Material=Get PipMaterial()) -> (then)
  ]}
 SetPipMaterial {nodes [
   n1 = SetPipMaterial!() -> (then, PipMaterial:Material Interface Object Reference) {is=title K2Node_FunctionEntry}
   > Set PipMaterial(PipMaterial:Material Interface Object Reference=@n1) -> (Output_Get:Material Interface Object Reference)
  ]}
]
variables [PipMaterial {type=MaterialInterface, default=null}]
widgets {
 info {
  inheritedWidgetCount=0
  parentClass=/Game/Athena/HUD/reticles/ReticleExtensionWidgets/ChargeStack/WBP_Reticle_RechargeStackBase_PipCollection.WBP_Reticle_RechargeStackBase_PipCollection_C
  rootWidgetClass=&DynamicEntryBox
  widgetCount=1
 }
 widgets [DynamicEntryBox_Stacks {
   widgetClassPath=&DynamicEntryBox
   bIsVariable=true
   widget=DynamicEntryBox_Stacks
   properties {
    entryWidgetClass=&/WBP_Reticle_Katana_Pip.WBP_Reticle_Katana_Pip_C
    numDesignerPreviewEntries=4
    entrySpacing=<-2 0>
    radialBoxSettings {angleBetweenItems=25, bClockwiseOrder=false, bDistributeItemsEvenly=false, marginSize=0, sectorCentralAngle=360, startingAngle=270}
   }
  }]
}
