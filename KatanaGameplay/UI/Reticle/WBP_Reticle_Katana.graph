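# Katana weapon reticle widget, extending the recharge-stack reticle base to show dash targeting and charge state.
# It binds to the katana's dash preview event, driving dynamic material parameters for lock-on to damageable non-team pawns and out-of-range blur and opacity.
# Ammo counts act as charges: gains play a sound, and a cooldown timer runs below a full magazine.
# Cooldown updates feed remaining recharge time to the pip collection.
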WidgetBlueprint /KatanaGameplay/UI/Reticle/WBP_Reticle_Katana
graphs [
 BindToWeaponEvents {nodes [
   n1 = BindToWeaponEvents!() -> (then, DoBind:bool, EquippedWeapon:Fort Weapon Object Reference) {is=title K2Node_FunctionEntry}
   > Parent: BindToWeaponEvents!(execute, DoBind:bool=@n1, EquippedWeapon:Fort Weapon Object Reference=@n1.EquippedWeapon) -> (then) {is=title K2Node_CallParentFunction}
   > n2 = Cast To B_Melee_Katana!(execute, Object=@n1.EquippedWeapon) -> (then, CastFailed, AsB Melee Katana:B Melee Katana Object Reference) {is=title K2Node_DynamicCast}
   > Set KatanaWeaponRef(KatanaWeaponRef:B Melee Katana Object Reference=@n2) -> (Output_Get:B Melee Katana Object Reference)
   > n3 = Branch(Condition=Get DoBind() -> (DoBind:bool))
   n4 = Create Event()
   n5 = Get KatanaWeaponRef() -> (KatanaWeaponRef:B Melee Katana Object Reference)
   > Bind Event to On Dash Preview Updated(self:B Melee Katana Object Reference=@n5, Delegate=@n4) {is=title K2Node_AddDelegate}
   Unbind Event from On Dash Preview Updated(execute=@n3.else, self:B Melee Katana Object Reference=@n5, Delegate=@n4) {is=title K2Node_RemoveDelegate}
  ]}
 EventGraph {nodes [
   Event Construct()
   > Parent: Construct()
   > UpdateCooldownQueryTimer(self:Fort Weapon Reticle Extension Widget Base Object Reference)
   > ShowLockedOn!(execute, self, IsLockedOn:bool=false) -> (then)
   Event Destruct()
   > Parent: Destruct()
   > HandleCooldownTimer!(execute, self:Fort Weapon Reticle Extension Widget Base Object Reference, bActivateTimer:bool=false, TimerInterval:float=0.25) -> (then)
   n1 = Event PreConstruct()
   > Parent: PreConstruct(IsDesignTime=@n1.IsDesignTime)
   > SetBrushFromMaterial(
    self=Get ReticleVisual() -> (ReticleVisual:Image Object Reference)
    Material=Get ReticleMaterial() -> (ReticleMaterial:Material Interface Object Reference)
   )
   > SetPipMaterial!(
    execute
    self:WBP Reticle Katana Pip Collection Object Reference=Get WBP_Reticle_Katana_PipCollection() -> (WBP_Reticle_Katana_PipCollection:WBP Reticle Katana Pip Collection Object Reference)
    PipMaterial:Material Interface Object Reference=Get PipMaterial() -> (PipMaterial:Material Interface Object Reference)
   ) -> (then)
   > SetRenderOpacity(self=Get WBP_Reticle_Katana_PipCollection(), InOpacity=Get PipCollectionRenderOpacity() -> (PipCollectionRenderOpacity:float))
   n2 = Event OnLocalAmmoChanged!() -> (OutputDelegate, then, LocalCount:int=0, LocalRemaining:int=0) {is=title K2Node_Event}
   > Parent: OnLocalAmmoChanged!(execute, LocalCount:int=@n2.LocalCount, LocalRemaining:int=@n2.LocalRemaining) -> (then) {is=title K2Node_CallParentFunction}
   n3 = Get PrevRemaining() -> (PrevRemaining:int=0)
   > n4 = Branch(Condition=AND Boolean(A="integer < integer"(A=@n3, B=@n2.LocalCount), B="Not Equal (Integer)"(A=@n2.LocalCount, B=@n3)))
   > n5 = Get!(execute:Exec) -> (then:Exec, else:Exec, KatanaWeaponRef:B Melee Katana Object Reference)
   > n6 = Play Sound Simple!(
    execute
    self:Sound Library Subsystem Object Reference=Get SoundLibrarySubsystem() -> (ReturnValue:Sound Library Subsystem Object Reference) {is=title K2Node_GetSubsystem}
    OwningActor:Actor Object Reference=@n5
    InSettings:"Simple Sound Settings Structure (by ref)"=Make Simple Sound Settings!(
     EventName:Gameplay Tag Structure=(TagName="SoundLibrary.Melee.Katana.ChargeAvailable")
     bAttachedSound:bool=False
     bStopWhenAttachedDestroyed:bool=False
     bAutoDestroy:bool=True
     bForceComponentCreation:bool=False
     VolumeMultiplier:float=1
     PitchMultiplier:float=1
     StartTime:float=0
     ConcurrencySettings:Sound Concurrency Object Reference
     AttenuationSettings:Sound Attenuation Object Reference
     AttachName:Name=None
     AttachToComponent:Scene Component Weak object
     LocationType:EAttachLocation Enum=KeepRelativeOffset
     Location:Vector="0,0,0"
    ) -> (SoundLibrarySimpleContextSettings:Simple Sound Settings Structure) {is=title K2Node_MakeStruct}
   ) -> (then, ReturnValue:Sound Library Play Result Structure)
   Set PrevRemaining(execute=[@n6.then, @n5.else, @n4.else], PrevRemaining:int=@n2.LocalCount) -> (Output_Get:int=0)
   > UpdateCooldownQueryTimer()
   Event UpdateCooldownQueryTimer() {is=title K2Node_Event}
   > n7 = Get()
   > n8 = Branch(Condition="integer >= integer"(
     A=GetLocalMagazineAmmoCount!(self:Fort Weapon Object Reference=@n7) -> (ReturnValue:int=0)
     B=GetBulletsPerClip!(self:Fort Weapon Object Reference=@n7) -> (ReturnValue:int=0)
    ))
   > HandleCooldownTimer(bActivateTimer=false, TimerInterval=0.25)
   HandleCooldownTimer(execute=@n8.else, bActivateTimer=true, TimerInterval=0.25)
   Event UpdateRemainingCooldown() {is=title K2Node_Event}
   > n9 = GetRemainingCooldownTimer!(execute, self) -> (then, RemainingTime:double, TotalDuration:double)
   > UpdateRemainingRechargeTime!(
    execute
    self:Fort Weapon Reticle Recharge Stack Base Pip Collection Base Object Reference=Get WBP_Reticle_Katana_PipCollection()
    TimeRemaining:float=@n9
    TotalDuration:float=@n9.TotalDuration
    PipPercentMatParamName:Name=Opacity
    ShouldSetPipColor:bool=false
    NotReadyColor:Linear Color Structure
    ReadyColor:Linear Color Structure
    PipColorMatParamName:Name=None
   ) -> (then)
  ]}
 GetPipCollection {nodes [
   GetPipCollection() {is=title K2Node_FunctionEntry}
   > Return Node!(execute, PipCollection:WBP Reticle Recharge Stack Base Pip Collection Object Reference=Get WBP_Reticle_Katana_PipCollection())
  ]}
 GetRemainingCooldownTimer {nodes [
   GetRemainingCooldownTimer() {is=title K2Node_FunctionEntry}
   > n1 = Cast To FortPlayerController(Object=GetOwningPlayer())
   > n2 = Get()
   n3 = GetCoolDownTimeRemaining!(
    self:Fort Item Object Reference=GetInventoryItem!(self:Fort Weapon Object Reference=@n2) -> (ReturnValue:Fort Item Object Reference)
    FortPC:Fort Player Controller Object Reference=@n1
   ) -> (CooldownDuration:float, ReturnValue:float)
   > Return Node!(execute, RemainingTime:double=@n3.ReturnValue, TotalDuration:double=@n3)
   Return Node(execute=[@n2.else, @n1.CastFailed], RemainingTime=0, TotalDuration=0)
  ]}
 OnDashPreviewUpdated {nodes [
   n1 = OnDashPreviewUpdated!() -> (
    then
    DashPreviewDirection:Vector=~
    DashPreviewDistance:double=~
    OutOfRangePercent:double=~
    DashPreviewBlockingActor:Actor Object Reference
    bWillDashPreviewDoDamage:bool=~
   ) {is=title K2Node_FunctionEntry}
   > n2 = Branch(Condition=@n1.bWillDashPreviewDoDamage)
   > n3 = Sequence()
   > n4 = Is Valid(InputObject=@n1.DashPreviewBlockingActor)
   > n5 = Branch(Condition=OnSameTeam!(
     ActorA:Actor Object Reference=GetOwningPlayerPawn!(self:User Widget Object Reference) -> (ReturnValue:Pawn Object Reference)
     ActorB:Actor Object Reference=@n1.DashPreviewBlockingActor
    ) -> (ReturnValue:bool))
   ShowLockedOn(execute=[@n5.then, @n4.Is Not Valid], IsLockedOn=false)
   ShowLockedOn(
    execute=@n5.else
    IsLockedOn=ClassIsChildOf(TestClass=Get Class(Object=@n1.DashPreviewBlockingActor), ParentClass=/Script/FortniteGame.FortPawn)
   )
   UpdateOutOfRangePercent!(execute=@n3.then_1, self, OutOfRangePercent:double=0, IsInRange:bool=true) -> (then)
   ShowLockedOn(execute=@n2.else, IsLockedOn=false)
   > UpdateOutOfRangePercent(OutOfRangePercent=@n1.OutOfRangePercent, IsInRange=false)
  ]}
 ShowLockedOn {nodes [
   n1 = ShowLockedOn!() -> (then, IsLockedOn:bool=~) {is=title K2Node_FunctionEntry}
   > n2 = Branch(Condition=@n1)
   n3 = Get ReticleVisual()
   > n4 = GetDynamicMaterial(self=@n3)
   > SetScalarParameterValue(self=@n4, ParameterName=LockedOn, Value=1)
   n5 = GetDynamicMaterial(execute=@n2.else, self=@n3)
   > SetScalarParameterValue(self=@n5, ParameterName=LockedOn, Value=0)
  ]}
 UpdateOutOfRangePercent {nodes [
   n1 = UpdateOutOfRangePercent!() -> (then, OutOfRangePercent:double, IsInRange:bool=~) {is=title K2Node_FunctionEntry}
   > n2 = GetDynamicMaterial(self=Get ReticleVisual())
   n3 = float - float(A=1, B=@n1)
   > SetScalarParameterValue(self=@n2, ParameterName=AnimBlur, Value=@n3)
   n4 = Get IsInRange() -> (IsInRange:bool)
   n5 = Get NearRangeMaxOpacity() -> (NearRangeMaxOpacity:float)
   > SetScalarParameterValue(
    self=@n2
    ParameterName=Opacity
    Value=MapRangeClamped(
     Value=@n3
     InRangeA=0
     InRangeB=1
     OutRangeA=Get OutOfRangeOpacity() -> (OutOfRangeOpacity:double)
     OutRangeB=Select(Option 0:float=@n5, Option 1:float=1, Index=@n4) -> (ReturnValue:float)
    )
   )
  ]}
]
variables [
 KatanaWeaponRef {type=B_Melee_Katana_C, default=null}
 NearRangeMaxOpacity {type=number, default=1}
 OutOfRangeOpacity {type=number, default=0.7}
 PipCollectionRenderOpacity {type=number, default=0.8}
 PipMaterial {type=MaterialInterface, default=null}
 PrevRemaining {type=integer, default=10}
 RemainingCooldownQueryTimer {type=TimerHandle}
 ReticleMaterial {type=MaterialInterface, default=&/MI_UI_KatanaReticleBase}
]
widgets {
 info {
  inheritedWidgetCount=1
  namedSlotCount=1
  parentClass=/Game/Athena/HUD/reticles/ReticleExtensionWidgets/ChargeStack/WBP_Reticle_RechargeStackBase.WBP_Reticle_RechargeStackBase_C
  rootWidgetClass=&Overlay
  widgetCount=5
 }
 widgets [
  Reticle {widgetClassPath=&Overlay, widget=Reticle}
  ReticleVisual {
   widgetClassPath=&Image
   bIsVariable=true
   parent=Reticle
   widget=ReticleVisual
   properties {
    brush {
     bIsDynamicallyLoaded=false
     drawAs=Image
     imageType=NoImage
     mirroring=NoMirror
     resourceName=None
     resourceObject=&MI_UI_KatanaReticleBase
     tiling=NoTile
     imageSize=<128 128>
     margin=<ltrb 0 0 0 0>
     outlineSettings {
      bUseBrushTransparency=false
      roundingType=HalfHeightRadius
      width=0
      color {colorUseRule=UseColor_Specified, specifiedColor=<rgba 0 0 0 0>}
      cornerRadii=<0 0 0 0>
     }
     tintColor {colorUseRule=UseColor_Specified, specifiedColor=<rgba 1 1 1 1>}
     uVRegion {bIsValid=false, max=<0 0>, min=<0 0>}
    }
    renderTransform {angle=0, scale=<0.75 0.75>, shear=<0 0>, translation=<0 9>}
   }
   slot {class=OverlaySlot}
  }
  FortVisualAttachment_0 {
   widgetClassPath=&FortVisualAttachment
   parent=Reticle
   widget=FortVisualAttachment_0
   properties {contentAnchor=<0 0.5>, renderTransform {angle=0, scale=<1 1>, shear=<0 0>, translation=<8 32>}}
   slot {class=OverlaySlot, properties {horizontalAlignment=HAlign_Right, verticalAlignment=VAlign_Center}}
  }
  WBP_Reticle_Katana_PipCollection {
   widgetClassPath=&/WBP_Reticle_Katana_PipCollection
   bIsVariable=true
   parent=FortVisualAttachment_0
   widget=WBP_Reticle_Katana_PipCollection
   properties {renderTransform {angle=0, scale=<1 1>, shear=<0 0>, translation=<9 16>}}
   slot {class=SizeBoxSlot}
  }
  TutorialHighlight {widgetClassPath=&TutorialHighlight}
 ]
}
